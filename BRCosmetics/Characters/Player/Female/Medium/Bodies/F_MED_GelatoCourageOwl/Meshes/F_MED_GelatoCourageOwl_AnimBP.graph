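# Post-process style animation Blueprint for the GelatoCourageOwl female medium body cosmetic, layering secondary adjustments over the pose copied from the attached parent mesh.
# Weighted Constraint nodes drive dynamic shoulder-setup bones, then a character-specific Control Rig finishes the pose.
# Each update converts weapon type, crouch, parachute, celebration montage and directional movement into mostly float weights, which feed that rig alongside skydive and emote flags.
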AnimBlueprint /BRCosmetics/Characters/Player/Female/Medium/Bodies/F_MED_GelatoCourageOwl/Meshes/F_MED_GelatoCourageOwl_AnimBP
target_skeleton=&/F_MED_GelatoCourageOwl_Skeleton
graphs [
 AnimGraph {nodes [
   n1 = "Constraint - dyn_shouldersetup_handle (PPPPPPPP)"!(
    ComponentPose=Local To Component(LocalPose=Copy Pose From Mesh() {bUseAttachedParent=true})
    Alpha=1
    ConstraintWeights_0=0.0286
    ConstraintWeights_1:float=0.0423
    ConstraintWeights_2:float=0.7213
    ConstraintWeights_3:float=0.0619
    ConstraintWeights_4:float=0.0575
    ConstraintWeights_5:float=0.0602
    ConstraintWeights_6:float=0.0001
    ConstraintWeights_7:float=0.0282
   ) -> (Pose) {
    alphaBoolBlend {blendOption=Linear}
    boneToModify {boneName=dyn_shouldersetup_handle}
    constraintSetup [
     {targetBone {boneName=spine_04}}
     {targetBone {boneName=spine_03}}
     {targetBone {boneName=spine_05}}
     {targetBone {boneName=clavicle_l}}
     {targetBone {boneName=clavicle_r}}
     {targetBone {boneName=neck_01}}
     {targetBone {boneName=head}}
     {targetBone {boneName=neck_02}}
    ]
    constraintWeights [1,1,1,1,1,1,1,1]
   }
   n2 = "Constraint - dyn_shouldersetup_l_back (PPPPP)"!(
    ComponentPose=@n1
    Alpha=1
    ConstraintWeights_0=0.0698
    ConstraintWeights_1:float=0.044
    ConstraintWeights_2:float=0.764
    ConstraintWeights_3:float=0.031
    ConstraintWeights_4:float=0.0911
   ) -> (Pose) {
    alphaBoolBlend {blendOption=Linear}
    boneToModify {boneName=dyn_shouldersetup_l_back}
    constraintSetup [
     {targetBone {boneName=upperarm_twist_01_l}}
     {targetBone {boneName=spine_05}}
     {targetBone {boneName=clavicle_l}}
     {targetBone {boneName=deform_upperarm_bk_l}}
     {targetBone {boneName=deform_upperarm_up_l}}
    ]
    constraintWeights [1,1,1,1,1]
   }
   n3 = "Constraint - dyn_shouldersetup_r_back (PPPPPPP)"!(
    ComponentPose=@n2
    Alpha=1
    ConstraintWeights_0=0.0883
    ConstraintWeights_1:float=0.0834
    ConstraintWeights_2:float=0.2048
    ConstraintWeights_3:float=0.455
    ConstraintWeights_4:float=0.0429
    ConstraintWeights_5:float=0.0972
    ConstraintWeights_6:float=0.0284
   ) -> (Pose) {
    alphaBoolBlend {blendOption=Linear}
    boneToModify {boneName=dyn_shouldersetup_r_back}
    constraintSetup [
     {targetBone {boneName=upperarm_twist_01_l}}
     {targetBone {boneName=spine_04}}
     {targetBone {boneName=spine_05}}
     {targetBone {boneName=clavicle_l}}
     {targetBone {boneName=deform_pec_l}}
     {targetBone {boneName=deform_upperarm_fr_l}}
     {targetBone {boneName=deform_upperarm_up_l}}
    ]
    constraintWeights [1,1,1,1,1,1,1]
   }
   n4 = "Constraint - dyn_shouldersetup_r_back (PPPPP)"!(
    ComponentPose=@n3
    Alpha=1
    ConstraintWeights_0=0.0597
    ConstraintWeights_1:float=0.781
    ConstraintWeights_2:float=0.0587
    ConstraintWeights_3:float=0.0248
    ConstraintWeights_4:float=0.0758
   ) -> (Pose) {
    alphaBoolBlend {blendOption=Linear}
    boneToModify {boneName=dyn_shouldersetup_r_back}
    constraintSetup [
     {targetBone {boneName=spine_05}}
     {targetBone {boneName=clavicle_r}}
     {targetBone {boneName=upperarm_twist_01_r}}
     {targetBone {boneName=deform_upperarm_bk_r}}
     {targetBone {boneName=deform_upperarm_up_r}}
    ]
    constraintWeights [1,1,1,1,1]
   }
   n5 = "Constraint - dyn_shouldersetup_r_front (PPPPPPP)"!(
    ComponentPose=@n4
    Alpha=1
    ConstraintWeights_0=0.0876
    ConstraintWeights_1:float=0.2243
    ConstraintWeights_2:float=0.4499
    ConstraintWeights_3:float=0.0799
    ConstraintWeights_4:float=0.0448
    ConstraintWeights_5:float=0.089
    ConstraintWeights_6:float=0.0244
   ) -> (Pose) {
    alphaBoolBlend {blendOption=Linear}
    boneToModify {boneName=dyn_shouldersetup_r_front}
    constraintSetup [
     {targetBone {boneName=spine_04}}
     {targetBone {boneName=spine_05}}
     {targetBone {boneName=clavicle_r}}
     {targetBone {boneName=upperarm_twist_01_r}}
     {targetBone {boneName=deform_pec_r}}
     {targetBone {boneName=deform_upperarm_fr_r}}
     {targetBone {boneName=deform_upperarm_up_r}}
    ]
    constraintWeights [1,1,1,1,1,1,1]
   }
   n6 = Control Rig!(
    Source=Component To Local(ComponentPose=@n5)
    Alpha
    Sky Dive_Floating:bool=Get bIsSkydiveFloating()
    Sky Dive Paraschute:bool=Get bIsParachuteOpened()
    Sky Dive Diving:bool=Get bIsSkydiveDiveMode()
    Get Rifle:double=Get Is Aiming Rifle() -> (Is Aiming Rifle:double)
    Pickaxe:double=Get Is Pickaxe() -> (Is Pickaxe:double)
    Get Pistol:double=Get Is Aiming Pistol() -> (Is Aiming Pistol:double)
    Crouching:bool=Get bIsCrouching()
    Moving2D:bool=Get bIsMoving2D()
    Emote:bool=Get bIsPlayingEmote()
    moving:double=Get moving2d() -> (moving2d:double)
    not moving:double=Get not moving2d() -> (not moving2d:double)
    not crouch:double=Get not crouching() -> (not crouching:double)
    sky dive:bool=Get bIsSkyDiving()
    Is_Parachuting:double=Get Is_Parachute_Open() -> (Is_Parachute_Open:double)
    Is_CelebratingEmote:double=Get IsCelebrationEmote() -> (IsCelebrationEmote:double)
    is Moving left:double=Get strafeLeft float() -> (strafeLeft float:double)
    is Moving right:double=Get strafeRight float() -> (strafeRight float:double)
    is Moving back:double=Get moveBackward() -> (moveBackward:double)
    is Moving Forward:double=Get moveForward() -> (moveForward:double)
   ) -> (Pose) {
    alphaBoolBlend {blendOption=Linear}
    controlRigAssetReference {blueprintRigClass=&/F_Med_GelatoCourageOwl_ControlRig.F_Med_GelatoCourageOwl_ControlRig_C}
   }
   Output Pose(Result=@n6) {layerGroup=None}
  ]}
 EventGraph {nodes [
   Event BlueprintUpdateAnimation()
   > n1 = Sequence!(execute) -> (then_0, then_1, then_2:Exec, then_3:Exec, then_4:Exec, then_5:Exec, then_6:Exec, then_7:Exec)
   > Gravity Override State Function DEPRECATED(CharacterPartAnimInstance=Self-Reference(), GameTimeInSeconds=GetGameTimeInSeconds(), GravityOverrideParams=Get Gravity Override Params())
   > Set Gravity Override(Gravity Override="0, 0, 0")
   n2 = Cast To FortPlayerAnimInstance(execute=@n1.then_1, Object=GetAnimInstance(self=GetOwnerSkeletalMeshComponent()))
   > Set Is Aiming Pistol(Is Aiming Pistol:double=0) -> (Output_Get:double)
   > Set Is Pickaxe(Is Pickaxe:double=0) -> (Output_Get:double)
   > Set Is Aiming Rifle(Is Aiming Rifle:double=0) -> (Output_Get:double)
   > Set Is Pickaxe Boo(Is Pickaxe Boo:bool=false) -> (Output_Get:bool)
   > n3 = Switch on EFortWeaponCoreAnimation!(execute, Selection:EFortWeaponCoreAnimation Enum=Get WeaponCoreAnim(self=@n2)) -> (
    Melee:Exec
    Pistol:Exec
    Shotgun:Exec
    PaperBlueprint:Exec
    Rifle:Exec
    MeleeOneHand:Exec
    MachinePistol:Exec
    RocketLauncher:Exec
    GrenadeLauncher:Exec
    GoingCommando:Exec
    AssaultRifle:Exec
    TacticalShotgun:Exec
    SniperRifle:Exec
    TrapPlacement:Exec
    ShoulderLauncher:Exec
    AbilityDecoTool:Exec
    Crossbow:Exec
    C4:Exec
    RemoteControl:Exec
    DualWield:Exec
    AR_BullPup:Exec
    AR_ForwardGrip:Exec
    MedPackPaddles:Exec
    SMG_P90:Exec
    AR_DrumGun:Exec
    Consumable_Small:Exec
    Consumable_Large:Exec
    Balloon:Exec
    MountedTurret:Exec
    CreativeTool:Exec
    ExplosiveBow:Exec
    AshtonIndigo:Exec
    AshtonChicago:Exec
    MeleeDualWield:Exec
    Unarmed:Exec
   )
   Branch(execute=[@n3.AssaultRifle, @n3.Crossbow, @n3.GrenadeLauncher, @n3.Rifle, @n3.RocketLauncher, @n3.Shotgun, @n3.SniperRifle, @n3.TacticalShotgun])
   > Set Is Aiming Rifle(Is Aiming Rifle=1)
   Branch(execute=[@n3.MachinePistol, @n3.Pistol])
   > Set Is Aiming Pistol(Is Aiming Pistol=1)
   Branch(execute=@n3.Melee)
   > Set Is Pickaxe(Is Pickaxe=1)
   > Set Is Pickaxe Boo(Is Pickaxe Boo=true)
   n4 = Get bIsMoving2D()
   n5 = Get bIsCrouching()
   Set moving2d(execute=@n1.then_2, moving2d:double="To Float (Boolean)"(InBool=AND Boolean(A=@n4, B=@n5))) -> (Output_Get:double)
   n6 = AND Boolean(A=Get bIsMoving2D(), B=Get bIsCrouching())
   Set not moving2d(execute=@n1.then_3, not moving2d:double="To Float (Boolean)"(InBool=NOT Boolean(A=@n6))) -> (Output_Get:double)
   n7 = Get bIsCrouching()
   Set not crouching(execute=@n1.then_4, not crouching:double="To Float (Boolean)"(InBool=NOT Boolean(A=@n7))) -> (Output_Get:double)
   n8 = GetAnimInstance(self=GetOwnerSkeletalMeshComponent())
   Is Valid(exec=@n1.then_6, InputObject=@n8)
   > n9 = Set Iscelebration(Iscelebration:bool=Montage_IsPlaying(self=@n8, Montage=/BRCosmetics/Animation/Game/MainPlayer/Emotes/Emote_Celebration_Female_M)) -> (Output_Get:bool)
   > n10 = Branch(Condition=@n9)
   > Set IsCelebrationEmote(IsCelebrationEmote:double=1) -> (Output_Get:double)
   Set IsCelebrationEmote(execute=@n10.else, IsCelebrationEmote=0)
   Set Is_Parachute_Open(execute=@n1.then_5, Is_Parachute_Open:double="To Float (Boolean)"(InBool=Get bIsParachuteOpened())) -> (Output_Get:double)
   n11 = PlayerMovement(execute=@n1.then_7, AnimInstance=Self-Reference(), XDotProductCoefficient=1, YDotProductCoefficient=1, ZDotProductCoefficient=1)
   n12 = BreakVector(InVec=MakeVector(X=@n11, Y=@n11.YMovement))
   > Set strafeLeft float(strafeLeft float:double="Absolute (Float)"(A="Clamp (Float)"(Value=@n12, Min=-1, Max=0))) -> (Output_Get:double)
   > Set strafeRight float(strafeRight float:double="Clamp (Float)"(Value=@n12, Min=0, Max=1)) -> (Output_Get:double)
   > Set moveForward(moveForward:double="Clamp (Float)"(Value=@n12.Y, Min=0, Max=1)) -> (Output_Get:double)
   > Set moveBackward(moveBackward:double="Absolute (Float)"(A="Clamp (Float)"(Value=@n12.Y, Min=-1, Max=0))) -> (Output_Get:double)
  ]}
]
variables [
 Gravity Override {type=Vector}
 Gravity Override Params {type=GravityOverrideParamsStruct}
 Is Aiming Pistol {type=number}
 Is Aiming Rifle {type=number}
 Is Crouch Sprinting {type=boolean}
 Is Pickaxe {type=number}
 Is Pickaxe Boo {type=boolean}
 Is_Parachute_Open {type=number}
 Iscelebration {type=boolean}
 IsCelebrationEmote {type=number}
 moveBackward {type=number}
 moveForward {type=number}
 moving2d {type=number}
 not crouching {type=number}
 not moving2d {type=number}
 strafeLeft float {type=number}
 strafeRight float {type=number}
]
