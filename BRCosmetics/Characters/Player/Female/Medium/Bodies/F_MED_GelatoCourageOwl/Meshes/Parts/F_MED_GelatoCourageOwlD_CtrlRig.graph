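# Control Rig for the GelatoCourageOwl female body part that drives three hair bone chains as the head moves.
# Parent constraints blend hair bones between head, neck and upper spine, then four spherical pose readers on the head write look up/down/left/right alphas.
# Up, down and left alphas then weight corrective transforms on hair bones.
# The right alpha is never consumed, and a separate chain of hair scaling nodes is unconnected.
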ControlRigBlueprint /BRCosmetics/Characters/Player/Female/Medium/Bodies/F_MED_GelatoCourageOwl/Meshes/Parts/F_MED_GelatoCourageOwlD_CtrlRig
graphs [
 ContainedGraph {nodes [
   n1 = RigVMFunctionEntryNode()
   SequenceExecution_2 = &RigVMFunction_Sequence(ExecuteContext=@n1.ExecuteContext)
   SequenceExecution_2_1 = &RigVMFunction_Sequence(ExecuteContext=@SequenceExecution_2.B)
   SequenceExecution_2_2 = &RigVMFunction_Sequence(ExecuteContext=@SequenceExecution_2_1.B)
   RigVMFunctionReturnNode(A=@SequenceExecution_2.A, B=@SequenceExecution_2_1.A, C=@SequenceExecution_2_2.A, D=@SequenceExecution_2_2.B)
  ]}
 RigVMFunctionLibrary {links [], nodes []}
 RigVMModel {nodes [
   &RigUnit_ModifyTransforms(
    ExecutePin=()
    ItemToModify="((Item=(Type=Bone,Name=\"asn1_hair_main_A_a\"),Transform=(Rotation=(X=0,Y=0,Z=0,W=1),Translation=(X=0,Y=0,Z=0),Scale3D=(X=0.9,Y=0.9,Z=0.9))),(Item=(Type=Bone,Name=\"asn1_hair_main_A_b\"),Transform=(Rotation=(X=0,Y=0,Z=0,W=1),Translation=(X=0,Y=0,Z=0),Scale3D=( ...(truncated)"
    Weight=1
    WeightMinimum=0
   )
   > &RigUnit_ModifyTransforms(
    ItemToModify="((Item=(Type=Bone,Name=\"asn1_hair_main_B_a\"),Transform=(Rotation=(X=0,Y=0,Z=0,W=1),Translation=(X=0,Y=0,Z=0),Scale3D=(X=0.9,Y=0.9,Z=0.9))),(Item=(Type=Bone,Name=\"asn1_hair_main_B_b\"),Transform=(Rotation=(X=0,Y=0,Z=0,W=1),Translation=(X=0,Y=0,Z=0),Scale3D=( ...(truncated)"
    Weight=1
    WeightMinimum=0
   )
   > &RigUnit_ModifyTransforms(
    ItemToModify="((Item=(Type=Bone,Name=\"asn1_hair_main_C_a\"),Transform=(Rotation=(X=0,Y=0,Z=0,W=1),Translation=(X=0,Y=0,Z=0),Scale3D=(X=0.9,Y=0.9,Z=0.9))),(Item=(Type=Bone,Name=\"asn1_hair_main_C_b\"),Transform=(Rotation=(X=0,Y=0,Z=0,W=1),Translation=(X=0,Y=0,Z=0),Scale3D=( ...(truncated)"
    Weight=1
    WeightMinimum=0
   )
   RigUnit_BeginExecution = &RigUnit_BeginExecution()
   RigVMFunction_Sequence = &RigVMFunction_Sequence(ExecuteContext=@RigUnit_BeginExecution.ExecutePin)
   &RigUnit_ParentConstraint(
    AdvancedSettings=(InterpolationType=Average,RotationOrderForFilter=XZY)
    Child=(Type=Bone,Name="asn1_hair_main_A_b")
    Parents=((Item=(Type=Bone,Name="neck_02"),Weight=0.4),(Item=(Type=Bone,Name="head"),Weight=0.6))
    Weight=1
    ExecutePin=@RigVMFunction_Sequence.A
   )
   > &RigUnit_ParentConstraint(
    AdvancedSettings=(InterpolationType=Average,RotationOrderForFilter=XZY)
    Child=(Type=Bone,Name="asn1_hair_main_A_c")
    Parents=((Item=(Type=Bone,Name="neck_01"),Weight=0.6),(Item=(Type=Bone,Name="head"),Weight=0.4))
    Weight=1
   )
   > &RigUnit_ParentConstraint(
    AdvancedSettings=(InterpolationType=Average,RotationOrderForFilter=XZY)
    Child=(Type=Bone,Name="asn1_hair_main_A_d")
    Parents=((Item=(Type=Bone,Name="head"),Weight=0.3),(Item=(Type=Bone,Name="spine_05"),Weight=0.7))
    Weight=1
   )
   > &RigUnit_ParentConstraint(
    AdvancedSettings=(InterpolationType=Average,RotationOrderForFilter=XZY)
    Child=(Type=Bone,Name="asn1_hair_main_B_b")
    Parents=((Item=(Type=Bone,Name="neck_02"),Weight=0.4),(Item=(Type=Bone,Name="head"),Weight=0.6))
    Weight=1
   )
   > &RigUnit_ParentConstraint(
    AdvancedSettings=(InterpolationType=Average,RotationOrderForFilter=XZY)
    Child=(Type=Bone,Name="asn1_hair_main_B_c")
    Parents=((Item=(Type=Bone,Name="neck_01"),Weight=0.6),(Item=(Type=Bone,Name="head"),Weight=0.4))
    Weight=1
   )
   > &RigUnit_ParentConstraint(
    AdvancedSettings=(InterpolationType=Average,RotationOrderForFilter=XZY)
    Child=(Type=Bone,Name="asn1_hair_main_B_d")
    Parents=((Item=(Type=Bone,Name="head"),Weight=0.3),(Item=(Type=Bone,Name="spine_05"),Weight=0.7))
    Weight=1
   )
   > &RigUnit_ParentConstraint(
    AdvancedSettings=(InterpolationType=Average,RotationOrderForFilter=XZY)
    Child=(Type=Bone,Name="asn1_hair_main_C_a")
    Parents=((Item=(Type=Bone,Name="neck_02"),Weight=0),(Item=(Type=Bone,Name="head"),Weight=1))
    Weight=1
   )
   > &RigUnit_ParentConstraint(
    AdvancedSettings=(InterpolationType=Average,RotationOrderForFilter=XZY)
    Child=(Type=Bone,Name="asn1_hair_main_C_b")
    Parents=((Item=(Type=Bone,Name="neck_01"),Weight=0.6),(Item=(Type=Bone,Name="head"),Weight=0.4))
    Weight=1
   )
   > &RigUnit_ParentConstraint(
    AdvancedSettings=(InterpolationType=Average,RotationOrderForFilter=XZY)
    Child=(Type=Bone,Name="asn1_hair_main_C_c")
    Parents=((Item=(Type=Bone,Name="head"),Weight=0.3),(Item=(Type=Bone,Name="spine_05"),Weight=0.7))
    Weight=1
   )
   > SequenceExecution_2 = RigVMAggregateNode()
   n1 = &RigUnit_SphericalPoseReader(
    ActiveRegionScaleFactors=(PositiveWidth=1,NegativeWidth=0.5,PositiveHeight=1.5,NegativeHeight=1.5)
    ActiveRegionSize=0.1
    Debug=(bDrawDebug=false,DebugScale=25,DebugSegments=20,DebugThickness=0.25)
    DriverAxis=(X=0,Y=1,Z=0)
    DriverItem=(Type=Bone,Name="head")
    FalloffRegionScaleFactors=(PositiveWidth=1,NegativeWidth=0.5,PositiveHeight=1.5,NegativeHeight=1.5)
    FalloffSize=0.2
    FlipWidthScaling=false
    OptionalParentItem=(Type=Bone)
    RotationOffset=(X=90,Y=0,Z=-25)
    ExecutePin=@SequenceExecution_2.A
   )
   > RigVMVariableNode(Value=@n1.OutputParam) {variable=head_SPR_look_up_alpha}
   n2 = &RigUnit_SphericalPoseReader(
    ActiveRegionScaleFactors=(PositiveWidth=0.75,NegativeWidth=0.75,PositiveHeight=1,NegativeHeight=0.5)
    ActiveRegionSize=0.1
    Debug=(bDrawDebug=true,DebugScale=25,DebugSegments=20,DebugThickness=0.25)
    DriverAxis=(X=0,Y=1,Z=0)
    DriverItem=(Type=Bone,Name="head")
    FalloffRegionScaleFactors=(PositiveWidth=1,NegativeWidth=1,PositiveHeight=1,NegativeHeight=0.5)
    FalloffSize=0.2
    OptionalParentItem=(Type=Bone)
    RotationOffset=(X=120,Y=0,Z=0)
    ExecutePin=@SequenceExecution_2.C
   )
   > RigVMVariableNode(Value=@n2.OutputParam) {variable=head_SPR_look_left_alpha}
   n3 = &RigUnit_SphericalPoseReader(
    ActiveRegionScaleFactors=(PositiveWidth=0.75,NegativeWidth=0.75,PositiveHeight=1,NegativeHeight=0.5)
    ActiveRegionSize=0.1
    Debug=(bDrawDebug=false,DebugScale=25,DebugSegments=20,DebugThickness=0.25)
    DriverAxis=(X=0,Y=1,Z=0)
    DriverItem=(Type=Bone,Name="head")
    FalloffRegionScaleFactors=(PositiveWidth=1,NegativeWidth=1,PositiveHeight=1,NegativeHeight=0.5)
    FalloffSize=0.2
    FlipHeightScaling=true
    FlipWidthScaling=false
    OptionalParentItem=(Type=Bone)
    RotationOffset=(X=60,Y=0,Z=0)
    ExecutePin=@SequenceExecution_2.D
   )
   > RigVMVariableNode(Value=@n3.OutputParam) {variable=head_SPR_look_right_alpha}
   n4 = &RigUnit_SphericalPoseReader(
    ActiveRegionScaleFactors=(PositiveWidth=2,NegativeWidth=0.5,PositiveHeight=2,NegativeHeight=2)
    ActiveRegionSize=0.075
    Debug=(bDrawDebug=false,DebugScale=25,DebugSegments=20,DebugThickness=0.25)
    DriverAxis=(X=0,Y=1,Z=0)
    DriverItem=(Type=Bone,Name="head")
    FalloffRegionScaleFactors=(PositiveWidth=1,NegativeWidth=0.4,PositiveHeight=1,NegativeHeight=1)
    FalloffSize=0.2
    FlipHeightScaling=false
    FlipWidthScaling=true
    OptionalParentItem=(Type=Bone)
    RotationOffset=(X=90,Y=0,Z=12.5)
    ExecutePin=@SequenceExecution_2.B
   )
   > RigVMVariableNode(Value=@n4.OutputParam) {variable=head_SPR_look_down_alpha}
   &RigUnit_ModifyTransforms(
    ItemToModify="((Item=(Type=Bone,Name=\"asn1_hair_main_C_a_base\"),Transform=(Rotation=(X=0,Y=-0.1736,Z=0,W=0.9848),Translation=(X=0,Y=0,Z=0),Scale3D=(X=1,Y=1,Z=1))),(Item=(Type=Bone,Name=\"asn1_hair_main_B_a_base\"),Transform=(Rotation=(X=0,Y=0,Z=-0.1305,W=0.9914),Translati ...(truncated)"
    WeightMinimum=0
    ExecutePin=@RigVMFunction_Sequence.B
    Weight=RigVMVariableNode() {variable=head_SPR_look_up_alpha}.Value
   )
   > &RigUnit_ModifyTransforms(
    ItemToModify="((Item=(Type=Bone,Name=\"asn1_hair_main_A_a_base\"),Transform=(Rotation=(X=0,Y=0,Z=0.1305,W=0.9914),Translation=(X=0,Y=0,Z=0),Scale3D=(X=1,Y=1,Z=1))),(Item=(Type=Bone,Name=\"asn1_hair_main_B_a_base\"),Transform=(Rotation=(X=0,Y=0,Z=0.0872,W=0.9962),Translation ...(truncated)"
    WeightMinimum=0
    Weight=RigVMVariableNode() {variable=head_SPR_look_down_alpha}.Value
   )
   > &RigUnit_ModifyTransforms(
    ItemToModify="((Item=(Type=Bone,Name=\"asn1_hair_main_A_a_base\"),Transform=(Rotation=(X=0,Y=0.0436,Z=0,W=0.999),Translation=(X=0,Y=0,Z=0),Scale3D=(X=1,Y=1,Z=1))),(Item=(Type=Bone,Name=\"asn1_hair_main_B_a_base\"),Transform=(Rotation=(X=0,Y=0.0436,Z=0,W=0.999),Translation=( ...(truncated)"
    WeightMinimum=0
    Weight=RigVMVariableNode() {variable=head_SPR_look_left_alpha}.Value
   )
  ]}
]
hierarchy [
 root {loc=<0 0 0>, rot=<0 0 0 1>}
 pelvis {parent=root, loc=<0 -0.5301 98.0571>, rot=<0.0127 0.707 -0.0127 -0.707>}
 spine_01 {parent=pelvis, loc=<6.1208 -0.9328 0>, rot=<0 0 0.034 -0.9994>}
 spine_02 {parent=spine_01, loc=<6.5996 -0.2095 0>, rot=<0 0 0.0238 -0.9997>}
 spine_03 {parent=spine_02, loc=<6.4914 -0.1237 0>, rot=<0 0 0.0149 -0.9999>}
 spine_04 {parent=spine_03, loc=<6.9598 0.0794 0>, rot=<0 0 -0.0132 -0.9999>}
 spine_05 {parent=spine_04, loc=<6.9987 0 0>, rot=<0 0 -0.0015 -1>}
 clavicle_l {parent=spine_05, loc=<5.4831 -0.1053 -3.2946>, rot=<0.0465 -0.7289 -0.009 -0.6829>}
 clavicle_r {parent=spine_05, loc=<5.4831 -0.1053 3.2946>, rot=<0.7289 0.0465 -0.6829 0.009>}
 neck_01 {parent=spine_05, loc=<11.3138 0.1529 0>, rot=<0 0 -0.2003 -0.9797>}
 neck_02 {parent=neck_01, loc=<4.5627 0 0>, rot=<0 0 0.0504 -0.9987>}
 head {parent=neck_02, loc=<5.3914 0 0>, rot=<0 0 0.115 -0.9934>}
 asn1_hair {parent=head, loc=<0.3282 -0.7473 0>, rot=<0 0 0 1>}
 asn1_hair_b {parent=asn1_hair, loc=<10.6128 2.6422 -8.0404>, rot=<0.0183 -0.9992 0.0047 0.0339>}
 asn1_hair_c {parent=asn1_hair, loc=<9.7316 -5.8031 -4.3772>, rot=<0.0183 -0.9992 0.0047 0.0339>}
 asn1_hair_main_A_a_base {parent=asn1_hair, loc=<5.186 3.8703 -7.9482>, rot=<0.0417 -0.9991 0 0>}
 asn1_hair_main_A_a {parent=asn1_hair_main_A_a_base, loc=<0 0 0>, rot=<0 0 0 1>}
 asn1_hair_main_A_b {parent=asn1_hair_main_A_a, loc=<6.1628 -0.406 -0.5216>, rot=<-0.0033 -0.034 -0.0234 -0.9991>}
 asn1_hair_main_A_c {parent=asn1_hair_main_A_b, loc=<5.2366 0.3478 -0.4818>, rot=<0.1348 0.0226 -0.0209 -0.9904>}
 asn1_hair_main_A_d {parent=asn1_hair_main_A_c, loc=<3.99 0.2397 -0.4252>, rot=<-0.0218 0.0554 -0.0639 -0.9962>}
 asn1_hair_a {parent=asn1_hair_main_A_b, loc=<2.023 -1.4766 1.9258>, rot=<0 0 0 1>}
 asn1_hair_main_B_a_base {parent=asn1_hair, loc=<5.4705 -2.7776 -6.8245>, rot=<-0.0741 -0.9665 0.2451 -0.019>}
 asn1_hair_main_B_a {parent=asn1_hair_main_B_a_base, loc=<0 0 0>, rot=<0 0 0 1>}
 asn1_hair_main_B_b {parent=asn1_hair_main_B_a, loc=<6.5035 0.9218 -0.967>, rot=<0.0602 0 0 -0.9982>}
 asn1_hair_main_B_c {parent=asn1_hair_main_B_b, loc=<4.9324 -0.2773 0.2271>, rot=<0.2491 -0.013 0.0269 -0.968>}
 asn1_hair_main_B_d {parent=asn1_hair_main_B_c, loc=<3.5268 0.0701 -0.234>, rot=<-0.013 -0.0659 -0.1025 -0.9925>}
 asn1_hair_main_C_a_base {parent=asn1_hair, loc=<4.611 -6.3393 -1.2898>, rot=<-0.1913 -0.7669 0.6106 -0.0493>}
 asn1_hair_main_C_a {parent=asn1_hair_main_C_a_base, loc=<0 0 0>, rot=<0 0 0 1>}
 asn1_hair_main_C_b {parent=asn1_hair_main_C_a, loc=<6.8698 0.445 -0.1526>, rot=<0.0577 0.0669 0.0507 -0.9948>}
 asn1_hair_main_C_c {parent=asn1_hair_main_C_b, loc=<4.4276 0.3046 0.0611>, rot=<0.1988 0.1165 0.0386 -0.9723>}
]
variables [
 head_SPR_look_down_alpha {type=double, default=0}
 head_SPR_look_left_alpha {type=double, default=0}
 head_SPR_look_right_alpha {type=double, default=0}
 head_SPR_look_up_alpha {type=double, default=0}
]
