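# Physics asset for the face accessory part of the female medium GelatoCourageOwl D cosmetic, providing bodies for its hair strands.
# Three capsule chains (A, B, C) branch from the hair root under the head, with narrow swing and tighter twist limits between segments.
# Head, neck, clavicle and upper-spine capsules, plus large neck boxes, form body volumes around the strands.
# Constraints among those body parts, hair root and strand bases leave swing and twist free.
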PhysicsAsset /BRCosmetics/Characters/Player/Female/Medium/Bodies/F_MED_GelatoCourageOwl/Meshes/Parts/F_MED_GelatoCourageOwlD_FaceAcc_Physics
bodies {
 asn1_hair [{shapeType=Capsule, length=1, radius=1}]
 asn1_hair_main_A_a [{shapeType=Capsule, length=3, radius=1.75, center=<3.105 0 0>, rotation=<rot 90 0 0>}]
 asn1_hair_main_A_a_base [{shapeType=Capsule, length=1, radius=1}]
 asn1_hair_main_A_b [{shapeType=Capsule, length=2.75, radius=1.5, center=<2.7376 0 0>, rotation=<rot 86.3848 0 0>}]
 asn1_hair_main_A_c [{shapeType=Capsule, length=1.7549, radius=1.125, center=<2.022 0 0>, rotation=<rot 90 0 0>}]
 asn1_hair_main_A_d [{shapeType=Capsule, length=1.8799, radius=1, center=<2.022 0 0>, rotation=<rot 90 0 0>}]
 asn1_hair_main_B_a [{shapeType=Capsule, length=2.875, radius=2.125, center=<3.0618 0 0>, rotation=<rot 90 0 0>}]
 asn1_hair_main_B_a_base [{shapeType=Capsule, length=1, radius=1}]
 asn1_hair_main_B_b [{shapeType=Capsule, length=2.25, radius=1.75, center=<2.5268 0 0>, rotation=<rot 90 0 0>}]
 asn1_hair_main_B_c [{shapeType=Capsule, length=2.875, radius=1.125, center=<2.5268 0 0>, rotation=<rot 90 0 0>}]
 asn1_hair_main_C_a [{shapeType=Capsule, length=3, radius=1.75, center=<3.4616 0 0.0001>, rotation=<rot 90 0 0>}]
 asn1_hair_main_C_a_base [{shapeType=Capsule, length=1, radius=1}]
 asn1_hair_main_C_b [{shapeType=Capsule, length=2.375, radius=1.125, center=<2.2691 0.0001 0>, rotation=<rot 90 0 0>}]
 asn1_hair_main_C_c [{shapeType=Capsule, length=2.75, radius=0.75, center=<2.2691 0.0001 0>, rotation=<rot 90 0 0>}]
 clavicle_l [{
   shapeType=Capsule
   length=5.2081
   radius=5.7147
   shapeName=clavicle_l_capsule
   center=<6.5243 -0.6953 0.104>
   rotation=<rot 16.9211 87.287 -90>
  }]
 clavicle_r [{
   shapeType=Capsule
   length=5.6159
   radius=5.6487
   shapeName=clavicle_r_capsule
   center=<-6.4341 0.3237 0.8599>
   rotation=<rot 3.2384 98.0696 -90>
  }]
 head [{shapeType=Capsule, length=5, radius=6.5, center=<2.4796 2.289 0.6971>, rotation=<rot 90 0 0>}]
 neck_01 [
  {shapeType=Capsule, length=4.25, radius=5, center=<2.166 0.4802 0>, rotation=<rot 90 0 0.8586>}
  {shapeType=Box, extentX=21.25, extentY=39, extentZ=8.5, shapeName=neck_01_box, center=<0.8879 4.2015 -13.4439>, rotation=<rot -15 -0.0001 0>}
 ]
 neck_02 [
  {
   shapeType=Box
   extentX=14.6971
   extentY=22.9547
   extentZ=8.8034
   shapeName=neck_02_box
   center=<-5.3184 -10.1281 0.1487>
   rotation=<rot 0.588 -30.1137 -89.2164>
  }
  {
   shapeType=Box
   extentX=17.605
   extentY=7.5
   extentZ=25.6754
   shapeName=neck_02_box
   center=<-3.536 -2.5243 -11.3565>
   rotation=<rot -12.5312 -17.7817 -56.1829>
  }
 ]
 spine_05 [{shapeType=Capsule, length=10, radius=7.75, center=<1.9244 -0.67 0>}]
}
constraints [
 {
  bone1Name=asn1_hair
  bone2Name=head
  swing1LimitDegrees=45
  swing1Motion=Free
  swing2LimitDegrees=45
  swing2Motion=Free
  twistLimitDegrees=45
  twistMotion=Free
 }
 {
  bone1Name=clavicle_l
  bone2Name=spine_05
  swing1LimitDegrees=45
  swing1Motion=Free
  swing2LimitDegrees=45
  swing2Motion=Free
  twistLimitDegrees=45
  twistMotion=Free
 }
 {
  bone1Name=clavicle_r
  bone2Name=spine_05
  swing1LimitDegrees=45
  swing1Motion=Free
  swing2LimitDegrees=45
  swing2Motion=Free
  twistLimitDegrees=45
  twistMotion=Free
 }
 {
  bone1Name=neck_01
  bone2Name=spine_05
  swing1LimitDegrees=45
  swing1Motion=Free
  swing2LimitDegrees=45
  swing2Motion=Free
  twistLimitDegrees=45
  twistMotion=Free
 }
 {
  bone1Name=neck_02
  bone2Name=neck_01
  swing1LimitDegrees=45
  swing1Motion=Free
  swing2LimitDegrees=45
  swing2Motion=Free
  twistLimitDegrees=45
  twistMotion=Free
 }
 {
  bone1Name=head
  bone2Name=neck_02
  swing1LimitDegrees=45
  swing1Motion=Free
  swing2LimitDegrees=45
  swing2Motion=Free
  twistLimitDegrees=45
  twistMotion=Free
 }
 {
  bone1Name=asn1_hair_main_A_a_base
  bone2Name=asn1_hair
  swing1LimitDegrees=45
  swing1Motion=Free
  swing2LimitDegrees=45
  swing2Motion=Free
  twistLimitDegrees=45
  twistMotion=Free
 }
 {bone1Name=asn1_hair_main_A_a, bone2Name=asn1_hair_main_A_a_base, swing1LimitDegrees=15, swing2LimitDegrees=15, twistLimitDegrees=5}
 {bone1Name=asn1_hair_main_A_b, bone2Name=asn1_hair_main_A_a, swing1LimitDegrees=15, swing2LimitDegrees=15, twistLimitDegrees=5}
 {bone1Name=asn1_hair_main_A_c, bone2Name=asn1_hair_main_A_b, swing1LimitDegrees=15, swing2LimitDegrees=15, twistLimitDegrees=5}
 {
  bone1Name=asn1_hair_main_B_a_base
  bone2Name=asn1_hair
  swing1LimitDegrees=45
  swing1Motion=Free
  swing2LimitDegrees=45
  swing2Motion=Free
  twistLimitDegrees=45
  twistMotion=Free
 }
 {bone1Name=asn1_hair_main_B_a, bone2Name=asn1_hair_main_B_a_base, swing1LimitDegrees=15, swing2LimitDegrees=15, twistLimitDegrees=5}
 {bone1Name=asn1_hair_main_B_b, bone2Name=asn1_hair_main_B_a, swing1LimitDegrees=15, swing2LimitDegrees=15, twistLimitDegrees=5}
 {
  bone1Name=asn1_hair_main_C_a_base
  bone2Name=asn1_hair
  swing1LimitDegrees=45
  swing1Motion=Free
  swing2LimitDegrees=45
  swing2Motion=Free
  twistLimitDegrees=45
  twistMotion=Free
 }
 {bone1Name=asn1_hair_main_C_a, bone2Name=asn1_hair_main_C_a_base, swing1LimitDegrees=15, swing2LimitDegrees=15, twistLimitDegrees=3}
 {bone1Name=asn1_hair_main_C_b, bone2Name=asn1_hair_main_C_a, swing1LimitDegrees=15, swing2LimitDegrees=15, twistLimitDegrees=3}
 {bone1Name=asn1_hair_main_B_c, bone2Name=asn1_hair_main_B_b, swing1LimitDegrees=15, swing2LimitDegrees=15, twistLimitDegrees=5}
 {bone1Name=asn1_hair_main_A_d, bone2Name=asn1_hair_main_A_c, swing1LimitDegrees=15, swing2LimitDegrees=15, twistLimitDegrees=5}
 {bone1Name=asn1_hair_main_C_c, bone2Name=asn1_hair_main_C_b, swing1LimitDegrees=15, swing2LimitDegrees=15, twistLimitDegrees=3}
]
