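# Animation Blueprint for the GelatoCourageOwl face accessory of a player cosmetic, driving its hair bones.
# Copied mesh poses are layered at the upper arms, then pass through a Control Rig and RigidBody simulation relative to spine_05.
# Afterwards, additive bone edits scale hair segments and rotate some by left-clavicle lift, measured each update against the reference pose.
# A head-yaw-driven hair-base correction chain exists but is not connected to the output.
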AnimBlueprint /BRCosmetics/Characters/Player/Female/Medium/Bodies/F_MED_GelatoCourageOwl/Meshes/Parts/F_MED_GelatoCourageOwlD_FaceAcc_AnimBP
target_skeleton=&/F_MED_GelatoCourageOwlD_FaceAcc_Skeleton
graphs [
 AnimGraph {nodes [
   n1 = Get headUp() -> (headUp:double)
   n2 = "Transform (Modify) Bone - Bone: asn1_hair_main_C_a_base"(Alpha=@n1, Translation="0,0,0", Rotation="5,0,0", Scale="1,1,1") {
    rotationMode=BMM_Additive
    rotationSpace=BCS_BoneSpace
    translationMode=BMM_Additive
    translationSpace=BCS_BoneSpace
    alphaBoolBlend {blendOption=Linear}
    alphaScaleBiasClamp {bMapRange=true, inRange {max=0, min=-50}, outRange {max=0, min=1}}
    boneToModify {boneName=asn1_hair_main_C_a_base}
    rotation {pitch=5}
   }
   n3 = "Transform (Modify) Bone - Bone: asn1_hair_main_B_a_base"(ComponentPose=@n2, Alpha=@n1, Translation="0,-1,0", Rotation="0,-5,0", Scale="1,1,1") {
    rotationMode=BMM_Additive
    rotationSpace=BCS_BoneSpace
    translationMode=BMM_Additive
    translationSpace=BCS_BoneSpace
    alphaBoolBlend {blendOption=Linear}
    alphaScaleBiasClamp {bMapRange=true, inRange {max=0, min=-50}, outRange {max=0, min=1}}
    boneToModify {boneName=asn1_hair_main_B_a_base}
    rotation {yaw=-5}
    translation {y=-1}
   }
   "Transform (Modify) Bone - Bone: asn1_hair_main_A_a_base"(ComponentPose=@n3, Alpha=@n1, Translation="0,0,0", Rotation="0,-3,0", Scale="1,1,1") {
    rotationMode=BMM_Additive
    rotationSpace=BCS_BoneSpace
    translationMode=BMM_Additive
    translationSpace=BCS_BoneSpace
    alphaBoolBlend {blendOption=Linear}
    alphaScaleBiasClamp {bMapRange=true, inRange {max=0, min=-50}, outRange {max=0, min=1}}
    boneToModify {boneName=asn1_hair_main_A_a_base}
    rotation {yaw=-3}
   }
   n4 = Get Skel_Mesh()
   n5 = Control Rig(Source=Layered blend per bone(
     BasePose=Copy Pose From Mesh() {bUseAttachedParent=true}
     BlendPoses_0=Copy Pose From Mesh(SourceMeshComponent=@n4) {bUseAttachedParent=true}
    ) {layerSetup [{branchFilters [{boneName=upperarm_l}, {boneName=upperarm_r}]}]}) {
    alphaBoolBlend {blendOption=Linear}
    controlRigAssetReference {blueprintRigClass=&/F_MED_GelatoCourageOwlD_CtrlRig.F_MED_GelatoCourageOwlD_CtrlRig_C}
   }
   n6 = RigidBody(ComponentPose=Local To Component(LocalPose=@n5)) {
    bClampLinearTranslationLimitToRefPose=true
    bOverrideWorldGravity=true
    overridePhysicsAsset=&/F_MED_GelatoCourageOwlD_FaceAcc_Physics
    simulationSpace=BaseBoneSpace
    alphaBoolBlend {blendOption=Linear}
    baseBoneRef {boneName=spine_05}
    componentLinearAccScale=<0.35 0.35 0.35>
    componentLinearVelScale=<0.35 0.35 0.35>
    overrideWorldGravity {z=-980}
   }
   n7 = "Transform (Modify) Bone - Bone: asn1_hair_main_A_b"(ComponentPose=@n6, Alpha=1, Translation="0,0,0", Rotation="0,0,0", Scale="0.9,0.9,0.9") {
    scaleMode=BMM_Additive
    scaleSpace=BCS_BoneSpace
    alphaBoolBlend {blendOption=Linear}
    boneToModify {boneName=asn1_hair_main_A_b}
    scale=<0.9 0.9 0.9>
   }
   n8 = "Transform (Modify) Bone - Bone: asn1_hair_main_A_c"(ComponentPose=@n7, Alpha=1, Translation="0,0,0", Rotation="0,0,0", Scale="0.9,0.9,0.9") {
    scaleMode=BMM_Additive
    scaleSpace=BCS_BoneSpace
    alphaBoolBlend {blendOption=Linear}
    boneToModify {boneName=asn1_hair_main_A_c}
    scale=<0.9 0.9 0.9>
   }
   n9 = "Transform (Modify) Bone - Bone: asn1_hair_main_B_b"(ComponentPose=@n8, Alpha=1, Translation="0,0,0", Rotation="0,0,0", Scale="0.9,0.9,0.9") {
    scaleMode=BMM_Additive
    scaleSpace=BCS_BoneSpace
    alphaBoolBlend {blendOption=Linear}
    boneToModify {boneName=asn1_hair_main_B_b}
    scale=<0.9 0.9 0.9>
   }
   n10 = "Transform (Modify) Bone - Bone: asn1_hair_main_B_c"(ComponentPose=@n9, Alpha=1, Translation="0,0,0", Rotation="0,0,0", Scale="0.9,0.9,0.9") {
    scaleMode=BMM_Additive
    scaleSpace=BCS_BoneSpace
    alphaBoolBlend {blendOption=Linear}
    boneToModify {boneName=asn1_hair_main_B_c}
    scale=<0.9 0.9 0.9>
   }
   n11 = "Transform (Modify) Bone - Bone: asn1_hair_main_C_b"(ComponentPose=@n10, Alpha=1, Translation="0,0,0", Rotation="0,0,0", Scale="0.9,0.9,0.9") {
    scaleMode=BMM_Additive
    scaleSpace=BCS_BoneSpace
    alphaBoolBlend {blendOption=Linear}
    boneToModify {boneName=asn1_hair_main_C_b}
    scale=<0.9 0.9 0.9>
   }
   n12 = "Transform (Modify) Bone - Bone: asn1_hair_main_C_c"(ComponentPose=@n11, Alpha=1, Translation="0,0,0", Rotation="0,0,0", Scale="0.8,0.8,0.8") {
    scaleMode=BMM_Additive
    scaleSpace=BCS_BoneSpace
    alphaBoolBlend {blendOption=Linear}
    boneToModify {boneName=asn1_hair_main_C_c}
    scale=<0.8 0.8 0.8>
   }
   n13 = Get clav l up() -> (clav l up:double)
   n14 = "Transform (Modify) Bone - Bone: asn1_hair_main_A_b"(ComponentPose=@n12, Alpha=@n13, Translation="0,0,0", Rotation="0,5,0", Scale="1,1,1") {
    rotationMode=BMM_Additive
    rotationSpace=BCS_BoneSpace
    alphaBoolBlend {blendOption=Linear}
    boneToModify {boneName=asn1_hair_main_A_b}
    rotation {yaw=5}
   }
   n15 = "Transform (Modify) Bone - Bone: asn1_hair_main_A_c"(ComponentPose=@n14, Alpha=@n13, Translation="0,0,0", Rotation="0,5,0", Scale="1,1,1") {
    rotationMode=BMM_Additive
    rotationSpace=BCS_BoneSpace
    alphaBoolBlend {blendOption=Linear}
    boneToModify {boneName=asn1_hair_main_A_c}
    rotation {yaw=5}
   }
   n16 = "Transform (Modify) Bone - Bone: asn1_hair_main_B_a"(ComponentPose=@n15, Alpha=@n13, Translation="0,0,0", Rotation="-10,0,0", Scale="1,1,1") {
    rotationMode=BMM_Additive
    rotationSpace=BCS_BoneSpace
    alphaBoolBlend {blendOption=Linear}
    boneToModify {boneName=asn1_hair_main_B_a}
    rotation {pitch=-10}
   }
   n17 = "Transform (Modify) Bone - Bone: asn1_hair_main_B_c"(ComponentPose=@n16, Alpha=@n13, Translation="0,0,0", Rotation="30,-20,0", Scale="1,1,1") {
    rotationMode=BMM_Additive
    rotationSpace=BCS_BoneSpace
    alphaBoolBlend {blendOption=Linear}
    boneToModify {boneName=asn1_hair_main_B_c}
    rotation {pitch=30, yaw=-20}
   }
   Output Pose(Result=Component To Local(ComponentPose=@n17)) {layerGroup=None}
  ]}
 EventGraph {nodes [
   Event BlueprintUpdateAnimation()
   > n1 = Sequence()
   n2 = GetDeltaTransformFromRefPose(self=GetOwnerSkeletalMeshComponent(), BoneName=head, BaseName=spine_05)
   > Set headUp(headUp:double=BreakRotator(InRot=BreakTransform(InTransform=@n2).Rotation).Yaw) -> (Output_Get:double)
   Set clav l up(
    execute=@n1.then_1
    clav l up:double=GetAlphaFromReferenceTransformDelta(
     SkeletalMeshComponent=Get Skel_Mesh()
     BoneName=clavicle_l
     BaseName=spine_05
     DeltaComponentType=EulerY
     InputRange=(X=15,Y=43)
     TwistAxis="1,0,0"
    )
   ) -> (Output_Get:double)
   Event BlueprintInitializeAnimation()
   > Set Skel_Mesh(Skel_Mesh=GetOwnerSkeletalMeshComponent())
   TryGetPawnOwner()
  ]}
]
variables [
 clav l up {type=number}
 headUp {type=number}
 Skel_Mesh {type=SkeletalMeshComponent}
]
