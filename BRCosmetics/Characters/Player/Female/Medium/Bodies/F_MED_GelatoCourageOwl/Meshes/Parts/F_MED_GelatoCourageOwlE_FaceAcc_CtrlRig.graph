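# Control Rig for the face-accessory part of the GelatoCourageOwl female cosmetic body, applying corrective offsets to hair bones based on head pose.
# Four spherical pose readers on the head bone write look up, down, left and right alphas into variables.
# The up alpha weights transform offsets on back hair bones, and the left alpha weights a rotation on dyn_left_hair_02.
# The down and right alphas are computed but drive nothing in this graph.
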ControlRigBlueprint /BRCosmetics/Characters/Player/Female/Medium/Bodies/F_MED_GelatoCourageOwl/Meshes/Parts/F_MED_GelatoCourageOwlE_FaceAcc_CtrlRig
graphs [
 ContainedGraph {nodes [
   n1 = RigVMFunctionEntryNode()
   SequenceExecution_2 = &RigVMFunction_Sequence(ExecuteContext=@n1.ExecuteContext)
   SequenceExecution_2_1 = &RigVMFunction_Sequence(ExecuteContext=@SequenceExecution_2.B)
   SequenceExecution_2_2 = &RigVMFunction_Sequence(ExecuteContext=@SequenceExecution_2_1.B)
   RigVMFunctionReturnNode(A=@SequenceExecution_2.A, B=@SequenceExecution_2_1.A, C=@SequenceExecution_2_2.A, D=@SequenceExecution_2_2.B)
  ]}
 RigVMFunctionLibrary {links [], nodes []}
 RigVMModel {nodes [
   RigUnit_BeginExecution = &RigUnit_BeginExecution()
   RigVMFunction_Sequence = &RigVMFunction_Sequence(ExecuteContext=@RigUnit_BeginExecution.ExecutePin)
   SequenceExecution_2 = RigVMAggregateNode(ExecuteContext=@RigVMFunction_Sequence.A)
   n1 = &RigUnit_SphericalPoseReader(
    ActiveRegionScaleFactors=(PositiveWidth=1,NegativeWidth=0.5,PositiveHeight=1.5,NegativeHeight=1.5)
    ActiveRegionSize=0.1
    Debug=(bDrawDebug=false,DebugScale=25,DebugSegments=20,DebugThickness=0.25)
    DriverAxis=(X=0,Y=1,Z=0)
    DriverItem=(Type=Bone,Name="head")
    FalloffRegionScaleFactors=(PositiveWidth=1,NegativeWidth=0.5,PositiveHeight=1.5,NegativeHeight=1.5)
    FalloffSize=0.2
    FlipWidthScaling=false
    OptionalParentItem=(Type=Bone)
    RotationOffset=(X=90,Y=0,Z=-25)
    ExecutePin=@SequenceExecution_2.A
   )
   > RigVMVariableNode(Value=@n1.OutputParam) {variable=head_SPR_look_up_alpha}
   n2 = &RigUnit_SphericalPoseReader(
    ActiveRegionScaleFactors=(PositiveWidth=0.75,NegativeWidth=0.75,PositiveHeight=1,NegativeHeight=0.5)
    ActiveRegionSize=0.1
    Debug=(bDrawDebug=false,DebugScale=25,DebugSegments=20,DebugThickness=0.25)
    DriverAxis=(X=0,Y=1,Z=0)
    DriverItem=(Type=Bone,Name="head")
    FalloffRegionScaleFactors=(PositiveWidth=1,NegativeWidth=1,PositiveHeight=1,NegativeHeight=0.5)
    FalloffSize=0.2
    OptionalParentItem=(Type=Bone)
    RotationOffset=(X=120,Y=0,Z=0)
    ExecutePin=@SequenceExecution_2.C
   )
   > RigVMVariableNode(Value=@n2.OutputParam) {variable=head_SPR_look_left_alpha}
   n3 = &RigUnit_SphericalPoseReader(
    ActiveRegionScaleFactors=(PositiveWidth=0.75,NegativeWidth=0.75,PositiveHeight=1,NegativeHeight=0.5)
    ActiveRegionSize=0.1
    Debug=(bDrawDebug=false,DebugScale=25,DebugSegments=20,DebugThickness=0.25)
    DriverAxis=(X=0,Y=1,Z=0)
    DriverItem=(Type=Bone,Name="head")
    FalloffRegionScaleFactors=(PositiveWidth=1,NegativeWidth=1,PositiveHeight=1,NegativeHeight=0.5)
    FalloffSize=0.2
    FlipHeightScaling=true
    FlipWidthScaling=false
    OptionalParentItem=(Type=Bone)
    RotationOffset=(X=60,Y=0,Z=0)
    ExecutePin=@SequenceExecution_2.D
   )
   > RigVMVariableNode(Value=@n3.OutputParam) {variable=head_SPR_look_right_alpha}
   n4 = &RigUnit_SphericalPoseReader(
    ActiveRegionScaleFactors=(PositiveWidth=2,NegativeWidth=0.5,PositiveHeight=2,NegativeHeight=2)
    ActiveRegionSize=0.075
    Debug=(bDrawDebug=false,DebugScale=25,DebugSegments=20,DebugThickness=0.25)
    DriverAxis=(X=0,Y=1,Z=0)
    DriverItem=(Type=Bone,Name="head")
    FalloffRegionScaleFactors=(PositiveWidth=1,NegativeWidth=0.4,PositiveHeight=1,NegativeHeight=1)
    FalloffSize=0.2
    FlipHeightScaling=false
    FlipWidthScaling=true
    OptionalParentItem=(Type=Bone)
    RotationOffset=(X=90,Y=0,Z=12.5)
    ExecutePin=@SequenceExecution_2.B
   )
   > RigVMVariableNode(Value=@n4.OutputParam) {variable=head_SPR_look_down_alpha}
   &RigUnit_ModifyTransforms(
    ItemToModify="((Item=(Type=Bone,Name=\"dyn_back_hair_MSTR\"),Transform=(Rotation=(X=0,Y=0,Z=0.0872,W=0.9962),Translation=(X=0,Y=1,Z=0),Scale3D=(X=1,Y=1,Z=1))),(Item=(Type=Bone,Name=\"dyn_back_hair_02\"),Transform=(Rotation=(X=0,Y=0,Z=0.0436,W=0.999),Translation=(X=0,Y=0,Z=0 ...(truncated)"
    WeightMinimum=0
    ExecutePin=@RigVMFunction_Sequence.B
    Weight=RigVMVariableNode() {variable=head_SPR_look_up_alpha}.Value
   )
   > &RigUnit_ModifyTransforms(
    ItemToModify=((Item=(Type=Bone,Name="dyn_left_hair_02"),Transform=(Rotation=(X=0,Y=0.0872,Z=0,W=0.9962),Translation=(X=0,Y=0,Z=0),Scale3D=(X=1,Y=1,Z=1))))
    WeightMinimum=0
    Weight=RigVMVariableNode() {variable=head_SPR_look_left_alpha}.Value
   )
  ]}
]
hierarchy [
 root {loc=<0 0 0>, rot=<0 0 0 1>}
 pelvis {parent=root, loc=<0 -0.5301 98.0571>, rot=<0.0127 0.707 -0.0127 -0.707>}
 spine_01 {parent=pelvis, loc=<6.1208 -0.9328 0>, rot=<0 0 0.034 -0.9994>}
 spine_02 {parent=spine_01, loc=<6.5996 -0.2095 0>, rot=<0 0 0.0238 -0.9997>}
 spine_03 {parent=spine_02, loc=<6.4914 -0.1237 0>, rot=<0 0 0.0149 -0.9999>}
 spine_04 {parent=spine_03, loc=<6.9598 0.0794 0>, rot=<0 0 -0.0132 -0.9999>}
 spine_05 {parent=spine_04, loc=<6.9987 0 0>, rot=<0 0 -0.0015 -1>}
 neck_01 {parent=spine_05, loc=<11.3138 0.1529 0>, rot=<0 0 -0.2003 -0.9797>}
 neck_02 {parent=neck_01, loc=<4.5627 0 0>, rot=<0 0 0.0504 -0.9987>}
 head {parent=neck_02, loc=<5.3914 0 0>, rot=<0 0 0.115 -0.9934>}
 dyn_left_hair_MSTR {parent=head, loc=<9.032 4.3476 -6.5366>, rot=<0.0041 -0.9906 -0.0006 -0.137>}
 dyn_left_hair_01 {parent=dyn_left_hair_MSTR, loc=<0 0 0>, rot=<0 0 0 1>}
 dyn_left_hair_02 {parent=dyn_left_hair_01, loc=<5.1819 -0.0769 0.1364>, rot=<0 -0.2552 0 -0.9669>}
 dyn_left_hair_f {parent=dyn_left_hair_02, loc=<1.9851 1.4606 0.3245>, rot=<0.974 -0.0022 0.0093 0.2263>}
 dyn_left_hair_fs1 {parent=dyn_left_hair_f, loc=<3.7711 -0.1261 0.8018>, rot=<0 0 0 1>}
 dyn_left_hair_fs2 {parent=dyn_left_hair_f, loc=<3.3888 1.5328 -0.8378>, rot=<0 0 0 1>}
 dyn_left_hair_m {parent=dyn_left_hair_02, loc=<6.3552 -1.9272 1.0718>, rot=<0.943 0.0024 -0.0087 -0.3326>}
 dyn_left_hair_b {parent=dyn_left_hair_02, loc=<4.1127 -4.4038 1.6538>, rot=<0.9997 0.0001 -0.0226 0.0047>}
 dyn_left_hair_bs1 {parent=dyn_left_hair_b, loc=<3.747 1.5802 -0.7781>, rot=<0 -0.0296 0 -0.9996>}
 dyn_back_hair_MSTR {parent=head, loc=<8.0626 -9.9146 0.1901>, rot=<-0.0101 -0.9979 0.0611 -0.0214>}
 dyn_back_hair_01 {parent=dyn_back_hair_MSTR, loc=<0 0 0>, rot=<0 0 0 1>}
 dyn_back_hair_02 {parent=dyn_back_hair_01, loc=<4.3256 0 -0.3691>, rot=<0 -0.0738 0 -0.9973>}
 dyn_back_hair_s1 {parent=dyn_back_hair_02, loc=<5.5583 1.0669 1.899>, rot=<0 0 0 1>}
 dyn_back_hair_s2 {parent=dyn_back_hair_02, loc=<4.3299 0.0866 -1.3696>, rot=<0 0 0 1>}
 dyn_back_hair_s3 {parent=dyn_back_hair_02, loc=<5.4639 3.7896 -3.9703>, rot=<0 0 0.0282 -0.9996>}
 dyn_right_hair_MSTR {parent=head, loc=<6.6401 -1.1803 8.1371>, rot=<0.005 -0.9985 0.0515 0.0176>}
 dyn_right_hair_01 {parent=dyn_right_hair_MSTR, loc=<0 0 0>, rot=<0 0 0 1>}
 dyn_right_hair_s1 {parent=dyn_right_hair_01, loc=<7.5379 2.1243 1.492>, rot=<0 0 0 1>}
 dyn_right_hair_s2 {parent=dyn_right_hair_01, loc=<7.4051 -0.208 -0.5682>, rot=<0 0 0 1>}
]
variables [
 head_SPR_look_down_alpha {type=double, default=0}
 head_SPR_look_left_alpha {type=double, default=0}
 head_SPR_look_right_alpha {type=double, default=0}
 head_SPR_look_up_alpha {type=double, default=0}
]
