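# Physics asset for the FaceAcc part of the female GelatoCourageOwlE character, simulating its dynamic back, left and right hair.
# Three hair chains hang from small MSTR bodies on the head, branching into box segments and capsule tip strands with limited swing and tight twist.
# Upper spine, neck and head bodies, including angled boxes around the neck, form the body volumes, linked to each other and to the chain roots by free angular constraints.
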PhysicsAsset /BRCosmetics/Characters/Player/Female/Medium/Bodies/F_MED_GelatoCourageOwl/Meshes/Parts/F_MED_GelatoCourageOwlE_FaceAcc_Physics
bodies {
 dyn_back_hair_01 [{shapeType=Box, extentX=5, extentY=2.25, extentZ=7, shapeName=dyn_back_hair_01_box, center=<0.25 0 0>}]
 dyn_back_hair_02 [{shapeType=Box, extentX=5.75, extentY=3, extentZ=9.75, shapeName=dyn_back_hair_02_box, center=<0.25 0 0>}]
 dyn_back_hair_MSTR [{shapeType=Capsule, length=0.505, radius=0.505, shapeName=dyn_back_hair_MSTR_capsule, rotation=<rot 0 -90 0>}]
 dyn_back_hair_s1 [{shapeType=Capsule, length=1.875, radius=1.125, shapeName=dyn_back_hair_s1_capsule, center=<1 0 0>, rotation=<rot 90 0 0>}]
 dyn_back_hair_s2 [{shapeType=Capsule, length=1.875, radius=1.125, shapeName=dyn_back_hair_s2_capsule, center=<1 0 0>, rotation=<rot 90 0 0>}]
 dyn_back_hair_s3 [{shapeType=Capsule, length=1.875, radius=1.125, shapeName=dyn_back_hair_s3_capsule, center=<1 0 0>, rotation=<rot 90 0 0>}]
 dyn_left_hair_01 [{shapeType=Box, extentX=5, extentY=7, extentZ=3, shapeName=dyn_left_hair_01_box, center=<0.25 0 0>}]
 dyn_left_hair_02 [{shapeType=Box, extentX=5, extentY=10.25, extentZ=2, shapeName=dyn_left_hair_02_box, center=<0.25 0 0>}]
 dyn_left_hair_MSTR [{shapeType=Capsule, length=0.505, radius=0.505, shapeName=dyn_left_hair_MSTR_capsule, rotation=<rot 0 -90 0>}]
 dyn_left_hair_b [{shapeType=Box, extentX=5, extentY=10.25, extentZ=2.25, shapeName=dyn_left_hair_b_box, center=<0.25 0 0>}]
 dyn_left_hair_bs1 [{shapeType=Capsule, length=2, radius=1, shapeName=dyn_left_hair_bs1_capsule, center=<1 0 0.5>, rotation=<rot 90 0 0>}]
 dyn_left_hair_f [{shapeType=Box, extentX=5, extentY=7, extentZ=2.25, shapeName=dyn_left_hair_f_box, center=<-0.75 0 0>}]
 dyn_left_hair_fs1 [{shapeType=Capsule, length=2, radius=1, shapeName=dyn_left_hair_fs1_capsule, center=<2.25 0 0>, rotation=<rot 90 0 0>}]
 dyn_left_hair_fs2 [{shapeType=Capsule, length=2, radius=1, shapeName=dyn_left_hair_fs2_capsule, center=<1.5 0 0>, rotation=<rot 90 0 0>}]
 dyn_left_hair_m [{shapeType=Box, extentX=5, extentY=7, extentZ=2, shapeName=dyn_left_hair_m_box, center=<0.25 0 0>, rotation=<rot 0 0 45>}]
 dyn_right_hair_01 [{shapeType=Box, extentX=5, extentY=7, extentZ=3.25, shapeName=dyn_right_hair_01_box, center=<0.5 0 0>, rotation=<rot 0 0 25>}]
 dyn_right_hair_MSTR [{shapeType=Capsule, length=0.505, radius=0.505, shapeName=dyn_right_hair_MSTR_capsule, rotation=<rot 0 -90 0>}]
 dyn_right_hair_s1 [{shapeType=Capsule, length=1.375, radius=1.625, shapeName=dyn_right_hair_s1_capsule, center=<1 0 -2>, rotation=<rot 90 0 0>}]
 dyn_right_hair_s2 [{shapeType=Capsule, length=2, radius=1, shapeName=dyn_right_hair_s2_capsule, center=<1 0 0>, rotation=<rot 90 0 0>}]
 head [{shapeType=Capsule, length=5, radius=6.5, center=<2.4796 2.289 0.6971>, rotation=<rot 90 0 0>}]
 neck_01 [
  {shapeType=Capsule, length=4.25, radius=5, center=<2.166 0.4802 0>, rotation=<rot 90 0 0.8586>}
  {
   shapeType=Box
   extentX=17.25
   extentY=7.5
   extentZ=23.75
   shapeName=neck_01_box
   center=<-1.1212 -5.4522 -8.7682>
   rotation=<rot -15.7835 -39.4401 -30.7364>
  }
  {
   shapeType=Box
   extentX=20.5
   extentY=26
   extentZ=7.5
   shapeName=neck_01_box
   center=<0.1987 -4.4091 10.1165>
   rotation=<rot 13.6127 -28.2869 -31.8122>
  }
 ]
 neck_02 [
  {
   shapeType=Box
   extentX=14.6971
   extentY=22.9547
   extentZ=8.8034
   shapeName=neck_02_box
   center=<-4.8964 -10.7905 0.1608>
   rotation=<rot 0.6541 -35.1137 -89.2707>
  }
  {
   shapeType=Box
   extentX=17.605
   extentY=7.5
   extentZ=25.6754
   shapeName=neck_02_box
   center=<-5.8744 -2.3096 -11.6217>
   rotation=<rot -16.6683 -20.6844 -55.4507>
  }
 ]
 spine_05 [{shapeType=Capsule, length=10, radius=7.75, center=<1.9244 -0.67 0>}]
}
constraints [
 {
  bone1Name=head
  bone2Name=neck_02
  swing1LimitDegrees=45
  swing1Motion=Free
  swing2LimitDegrees=45
  swing2Motion=Free
  twistLimitDegrees=45
  twistMotion=Free
 }
 {bone1Name=dyn_back_hair_01, bone2Name=dyn_back_hair_MSTR, swing1LimitDegrees=10, swing2LimitDegrees=10, twistLimitDegrees=2}
 {bone1Name=dyn_back_hair_02, bone2Name=dyn_back_hair_01, swing1LimitDegrees=10, swing2LimitDegrees=10, twistLimitDegrees=2}
 {bone1Name=dyn_back_hair_s1, bone2Name=dyn_back_hair_02, swing1LimitDegrees=10, swing2LimitDegrees=45, twistLimitDegrees=2}
 {bone1Name=dyn_back_hair_s2, bone2Name=dyn_back_hair_02, swing1LimitDegrees=10, swing2LimitDegrees=45, twistLimitDegrees=2}
 {bone1Name=dyn_back_hair_s3, bone2Name=dyn_back_hair_02, swing1LimitDegrees=10, swing2LimitDegrees=45, twistLimitDegrees=2}
 {bone1Name=dyn_left_hair_01, bone2Name=dyn_left_hair_MSTR, swing1LimitDegrees=12, swing2LimitDegrees=15, twistLimitDegrees=2}
 {bone1Name=dyn_left_hair_02, bone2Name=dyn_left_hair_01, swing1LimitDegrees=12, swing2LimitDegrees=15, twistLimitDegrees=2}
 {bone1Name=dyn_left_hair_b, bone2Name=dyn_left_hair_02, swing1LimitDegrees=15, swing2LimitDegrees=15, twistLimitDegrees=2}
 {bone1Name=dyn_left_hair_bs1, bone2Name=dyn_left_hair_b, swing1LimitDegrees=45, swing2LimitDegrees=45, twistLimitDegrees=2}
 {bone1Name=dyn_left_hair_f, bone2Name=dyn_left_hair_02, swing1LimitDegrees=15, swing2LimitDegrees=15, twistLimitDegrees=2}
 {bone1Name=dyn_left_hair_fs1, bone2Name=dyn_left_hair_f, swing1LimitDegrees=15, swing2LimitDegrees=15, twistLimitDegrees=2}
 {bone1Name=dyn_left_hair_fs2, bone2Name=dyn_left_hair_f, swing1LimitDegrees=15, swing2LimitDegrees=15, twistLimitDegrees=2}
 {bone1Name=dyn_left_hair_m, bone2Name=dyn_left_hair_02, swing1LimitDegrees=15, swing2LimitDegrees=15, twistLimitDegrees=2}
 {bone1Name=dyn_right_hair_01, bone2Name=dyn_right_hair_MSTR, swing1LimitDegrees=20, swing2LimitDegrees=15, twistLimitDegrees=2}
 {bone1Name=dyn_right_hair_s1, bone2Name=dyn_right_hair_01, swing1LimitDegrees=45, swing2LimitDegrees=45, twistLimitDegrees=2}
 {bone1Name=dyn_right_hair_s2, bone2Name=dyn_right_hair_01, swing1LimitDegrees=45, swing2LimitDegrees=45, twistLimitDegrees=2}
 {
  bone1Name=dyn_left_hair_MSTR
  bone2Name=head
  swing1LimitDegrees=45
  swing1Motion=Free
  swing2LimitDegrees=45
  swing2Motion=Free
  twistLimitDegrees=45
  twistMotion=Free
 }
 {
  bone1Name=dyn_back_hair_MSTR
  bone2Name=head
  swing1LimitDegrees=45
  swing1Motion=Free
  swing2LimitDegrees=45
  swing2Motion=Free
  twistLimitDegrees=45
  twistMotion=Free
 }
 {
  bone1Name=dyn_right_hair_MSTR
  bone2Name=head
  swing1LimitDegrees=45
  swing1Motion=Free
  swing2LimitDegrees=45
  swing2Motion=Free
  twistLimitDegrees=45
  twistMotion=Free
 }
 {
  bone1Name=neck_01
  bone2Name=spine_05
  swing1LimitDegrees=45
  swing1Motion=Free
  swing2LimitDegrees=45
  swing2Motion=Free
  twistLimitDegrees=45
  twistMotion=Free
 }
 {
  bone1Name=neck_02
  bone2Name=neck_01
  swing1LimitDegrees=45
  swing1Motion=Free
  swing2LimitDegrees=45
  swing2Motion=Free
  twistLimitDegrees=45
  twistMotion=Free
 }
]
