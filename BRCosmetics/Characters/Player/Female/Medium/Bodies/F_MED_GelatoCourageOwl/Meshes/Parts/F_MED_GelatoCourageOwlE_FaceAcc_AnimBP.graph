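# Post-body animation Blueprint for the GelatoCourageOwlE female face accessory, built on the Impulse Spring FaceAcc skeleton.
# Copies the attached parent's pose and curves through a Control Rig, then a spine_05-based RigidBody simulation with custom gravity.
# Each update derives neck alpha from head yaw, plus targeting and rigid-body floats, none visibly consumed in the AnimGraph.
# A second, head-based RigidBody node using Impulse Spring physics is left unconnected.
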AnimBlueprint /BRCosmetics/Characters/Player/Female/Medium/Bodies/F_MED_GelatoCourageOwl/Meshes/Parts/F_MED_GelatoCourageOwlE_FaceAcc_AnimBP
target_skeleton=/Game/Characters/Player/Female/Medium/Bodies/F_MED_Impulse_Spring/Skins/PER_01/Meshes/F_MED_Impulse_Spring_PER_1_FaceAcc_Skeleton
graphs [
 AnimGraph {nodes [
   RigidBody() {
    bClampLinearTranslationLimitToRefPose=true
    bOverrideWorldGravity=true
    overridePhysicsAsset=/Game/Characters/Player/Female/Medium/Bodies/F_MED_Impulse_Spring/Skins/PER_01/Meshes/F_MED_Impulse_Spring_PER_1_FaceAcc_Physics
    simulationSpace=BaseBoneSpace
    alphaBoolBlend {blendOption=Linear}
    baseBoneRef {boneName=head}
    componentAppliedLinearAccClamp=<3000 3000 3000>
    componentLinearAccScale=<0.3 0.3 0.35>
    componentLinearVelScale=<0.3 0.3 0.3>
    overrideWorldGravity {z=-800}
   }
   n1 = Control Rig(Source=Copy Pose From Mesh() {bCopyCurves=true, bUseAttachedParent=true}) {
    alphaBoolBlend {blendOption=Linear}
    controlRigAssetReference {blueprintRigClass=&/F_MED_GelatoCourageOwlE_FaceAcc_CtrlRig.F_MED_GelatoCourageOwlE_FaceAcc_CtrlRig_C}
   }
   n2 = RigidBody(ComponentPose=Local To Component(LocalPose=@n1)) {
    bClampLinearTranslationLimitToRefPose=true
    bOverrideWorldGravity=true
    overridePhysicsAsset=&/F_MED_GelatoCourageOwlE_FaceAcc_Physics
    simulationSpace=BaseBoneSpace
    alphaBoolBlend {blendOption=Linear}
    baseBoneRef {boneName=spine_05}
    componentLinearAccScale=<0.35 0.35 0.35>
    componentLinearVelScale=<0.35 0.35 0.35>
    overrideWorldGravity {z=-980}
   }
   Output Pose(Result=Component To Local(ComponentPose=@n2)) {layerGroup=None}
  ]}
 EventGraph {nodes [
   Event BlueprintUpdateAnimation()
   n1 = BreakTransform(InTransform=GetDeltaTransformFromRefPose(self=GetOwnerSkeletalMeshComponent(), BoneName=head, BaseName=spine_05))
   > Set Neck Alpha(Neck Alpha:double=MapRangeClamped(Value=BreakRotator(InRot=@n1.Rotation).Yaw, InRangeA=-5, InRangeB=-40)) -> (Output_Get:double)
   > Set ADS(ADS:double=SelectFloat(A=1, B=0, bPickA=Get bIsTargeting())) -> (Output_Get:double)
   > Set ADSoff(ADSoff:double=SelectFloat(A=0, B=1, bPickA=Get bIsTargeting())) -> (Output_Get:double)
   > Set RBAN_hair(RBAN_hair:double=SelectFloat(A=1, B=0, bPickA=Get bRigidBodyEnabled())) -> (Output_Get:double)
   TryGetPawnOwner()
  ]}
]
variables [
 ADS {type=number}
 ADSoff {type=number}
 Neck Alpha {type=number}
 RBAN_hair {type=number}
]
